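# Physics asset for the lower-body mesh of the male medium Pencil Athletic cosmetic character, covering a dynamic belt, pelvis and thighs.
# Capsule bodies form a three-segment belt chain hanging off a tiny master body, alongside a two-capsule pelvis and one capsule per thigh.
# The master body joins the pelvis with free swing and twist, while the thigh joints and belt links carry 45-degree angular limits.
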PhysicsAsset /BRCosmetics/Characters/Player/Male/Medium/Bodies/M_MED_Pencil_A/Athletic/Meshes/Pencil_Athletic_MedA_Lbdy_Physics
bodies {
 dyn_belt_01 [{shapeType=Capsule, length=15.9521, radius=5, shapeName=dyn_belt_01_capsule, center=<10 0 4>, rotation=<rot 0 0 90>}]
 dyn_belt_02 [{shapeType=Capsule, length=15.9521, radius=5, shapeName=dyn_belt_02_capsule, center=<10 0 4>, rotation=<rot 0 0 90>}]
 dyn_belt_03 [{shapeType=Capsule, length=15.9521, radius=5, shapeName=dyn_belt_03_capsule, center=<10 0 4>, rotation=<rot 0 0 90>}]
 dyn_belt_MSTR [{shapeType=Capsule, length=0.505, radius=0.505, shapeName=dyn_belt_MSTR_capsule, rotation=<rot 0 -90 0>}]
 pelvis [
  {shapeType=Capsule, length=27.3556, radius=9.0841}
  {shapeType=Capsule, length=25.3193, radius=11.1205, shapeName=pelvis_capsule, center=<-13.7115 1.3783 0>}
 ]
 thigh_l [{shapeType=Capsule, length=29.1395, radius=10.3589, center=<-20.1927 -1.1186 0>, rotation=<rot 90 180 176.4719>}]
 thigh_r [{shapeType=Capsule, length=20.1522, radius=9.89, center=<20.031 0 0>, rotation=<rot 90 0 0>}]
}
constraints [
 {bone1Name=thigh_l, bone2Name=pelvis, swing1LimitDegrees=45, swing2LimitDegrees=45, twistLimitDegrees=45}
 {bone1Name=thigh_r, bone2Name=pelvis, swing1LimitDegrees=45, swing2LimitDegrees=45, twistLimitDegrees=45}
 {
  bone1Name=dyn_belt_MSTR
  bone2Name=pelvis
  swing1LimitDegrees=45
  swing1Motion=Free
  swing2LimitDegrees=45
  swing2Motion=Free
  twistLimitDegrees=45
  twistMotion=Free
 }
 {bone1Name=dyn_belt_01, bone2Name=dyn_belt_MSTR, swing1LimitDegrees=45, swing2LimitDegrees=45, twistLimitDegrees=45}
 {bone1Name=dyn_belt_02, bone2Name=dyn_belt_01, swing1LimitDegrees=45, swing2LimitDegrees=45, twistLimitDegrees=45}
 {bone1Name=dyn_belt_03, bone2Name=dyn_belt_02, swing1LimitDegrees=45, swing2LimitDegrees=45, twistLimitDegrees=45}
]
