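# Physics asset for the front-end version of the Pencil Athletic medium male lower-body cosmetic mesh, covering a dynamic belt, pelvis and thighs.
# Capsules on the pelvis and both thighs form the body volumes around the belt chain.
# The belt's small master body hangs from the pelvis with free swing and twist, and three belt segments chain from it with angular limits, tighter in twist than swing.
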PhysicsAsset /BRCosmetics/Characters/Player/Male/Medium/Bodies/M_MED_Pencil_A/Athletic/Meshes/Pencil_Athletic_MedA_Lbdy_Physics_FrontEnd
bodies {
 dyn_belt_01 [{shapeType=Capsule, length=8.0242, radius=5.0207, shapeName=dyn_bearChain_01_capsule, center=<9.3346 0 6>, rotation=<rot 90 0 0>}]
 dyn_belt_02 [{shapeType=Capsule, length=2.7015, radius=4.9952, shapeName=dyn_bearChain_02_capsule, center=<3.4179 0 6>, rotation=<rot 90 0 0>}]
 dyn_belt_03 [{shapeType=Capsule, length=3.1668, radius=5.2798, shapeName=dyn_bearChain_02_capsule, center=<3.4179 0 6>, rotation=<rot 90 0 0>}]
 dyn_belt_MSTR [{shapeType=Capsule, length=0.505, radius=0.505, shapeName=dyn_belt_MSTR_capsule}]
 pelvis [
  {shapeType=Capsule, length=46.3466, radius=10.7105, center=<0 -2.6668 0>}
  {shapeType=Capsule, length=40.4223, radius=11.6425, shapeName=pelvis_capsule, center=<-13.4788 -0.4736 -0.0093>, rotation=<rot 0 0 1.9288>}
 ]
 thigh_l [{shapeType=Capsule, length=20.167, radius=10.7569, center=<-16.8615 0 0>, rotation=<rot 90 0 0>}]
 thigh_r [{shapeType=Capsule, length=20.924, radius=10, center=<16.0925 0 0>, rotation=<rot 90 0 0>}]
}
constraints [
 {bone1Name=thigh_l, bone2Name=pelvis, swing1LimitDegrees=45, swing2LimitDegrees=45, twistLimitDegrees=45}
 {bone1Name=thigh_r, bone2Name=pelvis, swing1LimitDegrees=45, swing2LimitDegrees=45, twistLimitDegrees=45}
 {
  bone1Name=dyn_belt_MSTR
  bone2Name=pelvis
  swing1LimitDegrees=45
  swing1Motion=Free
  swing2LimitDegrees=45
  swing2Motion=Free
  twistLimitDegrees=45
  twistMotion=Free
 }
 {bone1Name=dyn_belt_01, bone2Name=dyn_belt_MSTR, swing1LimitDegrees=45, swing2LimitDegrees=45, twistLimitDegrees=20}
 {bone1Name=dyn_belt_02, bone2Name=dyn_belt_01, swing1LimitDegrees=45, swing2LimitDegrees=45, twistLimitDegrees=35}
 {bone1Name=dyn_belt_03, bone2Name=dyn_belt_02, swing1LimitDegrees=45, swing2LimitDegrees=45, twistLimitDegrees=35}
]
