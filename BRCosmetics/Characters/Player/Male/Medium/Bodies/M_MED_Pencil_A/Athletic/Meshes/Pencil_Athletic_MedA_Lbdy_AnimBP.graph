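# Post-process style AnimBlueprint for the Pencil Athletic lower-body cosmetic, adding secondary motion to its dynamic belt bones.
# It copies the attached parent mesh's pose, then applies weighted bone constraints and RigidBody or AnimDynamics simulation.
# Branches switch on an upper-body Jacket meta tag, front end and crouching, using state-specific physics assets.
# Still and moving crouch flags toggle corrective bone offsets; the computed GravityOverride is unused in the graph.
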AnimBlueprint /BRCosmetics/Characters/Player/Male/Medium/Bodies/M_MED_Pencil_A/Athletic/Meshes/Pencil_Athletic_MedA_Lbdy_AnimBP
target_skeleton=&/Pencil_Athletic_MedA_Lbdy_Skeleton
graphs [
 AnimGraph {nodes [
   Input Pose()
   n1 = "Constraint - dyn_belt_MSTR (PPPPPP)"!(
    ComponentPose=Local To Component(LocalPose=Use cached pose 'ControlRig'())
    Alpha=1
    ConstraintWeights_0=0.7718
    ConstraintWeights_1:float=0.0573
    ConstraintWeights_2:float=0.0694
    ConstraintWeights_3:float=0.0347
    ConstraintWeights_4:float=0.0232
    ConstraintWeights_5:float=0.0437
   ) -> (Pose) {
    alphaBoolBlend {blendOption=Linear}
    boneToModify {boneName=dyn_belt_MSTR}
    constraintSetup [
     {targetBone {boneName=pelvis}}
     {targetBone {boneName=thigh_twist_01_l}}
     {targetBone {boneName=deform_groin_l}}
     {targetBone {boneName=spine_01}}
     {targetBone {boneName=spine_03}}
     {targetBone {boneName=spine_02}}
    ]
    constraintWeights [1,1,1,1,1,1]
   }
   n2 = "Constraint - dyn_belt_MSTR (R)"(ComponentPose=@n1, Alpha=1, ConstraintWeights_0=1) {
    alphaBoolBlend {blendOption=Linear}
    boneToModify {boneName=dyn_belt_MSTR}
    constraintSetup [{transformType=Rotation, targetBone {boneName=thigh_l}}]
    constraintWeights [1]
   }
   n3 = RigidBody!(ComponentPose=@n2, Alpha, OverrideWorldGravity:Vector="0,0,-980", ExternalForce) -> (Pose) {
    bClampLinearTranslationLimitToRefPose=true
    bOverrideWorldGravity=true
    overridePhysicsAsset=&/Pencil_Athletic_MedA_Lbdy_Physics_FrontEnd
    simulationSpace=BaseBoneSpace
    alphaBoolBlend {blendOption=Linear}
    baseBoneRef {boneName=root}
    componentAppliedLinearAccClamp=<5000 5000 5000>
    componentLinearAccScale=<0.4 0.4 0.5>
    componentLinearVelScale=<0.4 0.4 0.5>
    overrideWorldGravity {z=-980}
    simSpaceSettings {
     maxAngularAcceleration=1000
     maxAngularVelocity=3
     maxLinearAcceleration=700
     maxLinearVelocity=500
     velocityScaleZ=0
     externalLinearDragV=<0.2 0.2>
    }
   }
   n4 = "Constraint - dyn_belt_MSTR (PPPPPP)"(
    ComponentPose=Local To Component(LocalPose=Use cached pose 'ControlRig'())
    Alpha=1
    ConstraintWeights_0=0.7718
    ConstraintWeights_1=0.0573
    ConstraintWeights_2=0.0694
    ConstraintWeights_3=0.0347
    ConstraintWeights_4=0.0232
    ConstraintWeights_5=0.0437
   ) {
    alphaBoolBlend {blendOption=Linear}
    boneToModify {boneName=dyn_belt_MSTR}
    constraintSetup [
     {targetBone {boneName=pelvis}}
     {targetBone {boneName=thigh_twist_01_l}}
     {targetBone {boneName=deform_groin_l}}
     {targetBone {boneName=spine_01}}
     {targetBone {boneName=spine_03}}
     {targetBone {boneName=spine_02}}
    ]
    constraintWeights [1,1,1,1,1,1]
   }
   n5 = "Constraint - dyn_belt_MSTR (R)"(ComponentPose=@n4, Alpha=1, ConstraintWeights_0=1) {
    alphaBoolBlend {blendOption=Linear}
    boneToModify {boneName=dyn_belt_MSTR}
    constraintSetup [{transformType=Rotation, targetBone {boneName=thigh_twist_01_l}}]
    constraintWeights [1]
   }
   n6 = RigidBody(ComponentPose=@n5, OverrideWorldGravity="0,0,-980") {
    bClampLinearTranslationLimitToRefPose=true
    bOverrideWorldGravity=true
    overridePhysicsAsset=&/Pencil_Athletic_MedA_Lbdy_Physics
    simulationSpace=BaseBoneSpace
    alphaBoolBlend {blendOption=Linear}
    baseBoneRef {boneName=pelvis}
    componentAppliedLinearAccClamp=<2000 2000 2000>
    componentLinearAccScale=<0.4 0.4 0.45>
    overrideWorldGravity {z=-980}
    simSpaceSettings {
     maxAngularAcceleration=1000
     maxAngularVelocity=3
     maxLinearAcceleration=700
     maxLinearVelocity=500
     velocityScaleZ=0
     worldAlpha=1
     externalLinearDragV=<0.2 0.2>
    }
   }
   n7 = Blend Poses by bool(
    BlendPose_0=Component To Local(ComponentPose=@n3)
    BlendPose_1=Component To Local(ComponentPose=@n6)
    BlendTime_0=0.2
    BlendTime_1=0.2
    bActiveValue=Get bIsInFrontEnd()
   ) {blendType=Linear}
   n8 = "Constraint - dyn_belt_MSTR (PPPPPPPP)"!(
    ComponentPose=Local To Component(LocalPose=Use cached pose 'ControlRig'())
    bAlphaBoolEnabled:bool=True
    ConstraintWeights_0=0.2505
    ConstraintWeights_1:float=0.4852
    ConstraintWeights_2:float=0.0022
    ConstraintWeights_3:float=0.001
    ConstraintWeights_4:float=0.134
    ConstraintWeights_5:float=0.0615
    ConstraintWeights_6:float=0.0146
    ConstraintWeights_7:float=0.0511
   ) -> (Pose) {
    alphaInputType=Bool
    alphaBoolBlend {blendOption=Linear}
    boneToModify {boneName=dyn_belt_MSTR}
    constraintSetup [
     {targetBone {boneName=pelvis}}
     {targetBone {boneName=thigh_twist_01_l}}
     {targetBone {boneName=thigh_r}}
     {targetBone {boneName=deform_glute_l}}
     {targetBone {boneName=deform_groin_l}}
     {targetBone {boneName=thigh_twist_01_r}}
     {targetBone {boneName=thigh_l}}
     {targetBone {boneName=deform_groin_r}}
    ]
    constraintWeights [1,1,1,1,1,1,1,1]
   }
   n9 = "Transform (Modify) Bone - Bone: dyn_belt_MSTR"!(
    ComponentPose=@n8
    bAlphaBoolEnabled:bool=Get CrouchStill() -> (CrouchStill:bool)
    Translation="5.738,4.937,0.369"
    Rotation="-31.741,37.399,43.934"
    Scale="1,1,1"
   ) -> (Pose) {
    alphaInputType=Bool
    rotationMode=BMM_Additive
    rotationSpace=BCS_BoneSpace
    scaleMode=BMM_Additive
    scaleSpace=BCS_BoneSpace
    translationMode=BMM_Additive
    translationSpace=BCS_BoneSpace
    alphaBoolBlend {blendInTime=0.1, blendOption=Linear, blendOutTime=0.1}
    boneToModify {boneName=dyn_belt_MSTR}
    rotation=<rot -31.741 37.399 43.934>
    translation=<5.738 4.937 0.369>
   }
   n10 = "Transform (Modify) Bone - Bone: dyn_belt_MSTR"(
    ComponentPose=@n9
    bAlphaBoolEnabled=Get Moving Crouch() -> (Moving Crouch:bool)
    Translation="4.029,4.57,-1.051"
    Rotation="17.485,25.507,47.599"
    Scale="1,1,1"
   ) {
    alphaInputType=Bool
    rotationMode=BMM_Additive
    rotationSpace=BCS_BoneSpace
    scaleMode=BMM_Additive
    scaleSpace=BCS_BoneSpace
    translationMode=BMM_Additive
    translationSpace=BCS_BoneSpace
    alphaBoolBlend {blendInTime=0.1, blendOption=Linear, blendOutTime=0.1}
    boneToModify {boneName=dyn_belt_MSTR}
    rotation=<rot 17.485 25.507 47.599>
    translation=<4.029 4.57 -1.051>
   }
   n11 = RigidBody(ComponentPose=@n10, OverrideWorldGravity="0,0,-980") {
    bClampLinearTranslationLimitToRefPose=true
    bOverrideWorldGravity=true
    overridePhysicsAsset=&/Pencil_Athletic_MedA_Lbdy_Physics_Crouching
    simulationSpace=BaseBoneSpace
    alphaBoolBlend {blendOption=Linear}
    baseBoneRef {boneName=pelvis}
    componentAppliedLinearAccClamp=<500 500 500>
    componentLinearAccScale=<0.3 0.3 0.3>
    overrideWorldGravity {z=-980}
    simSpaceSettings {
     maxAngularAcceleration=1000
     maxAngularVelocity=3
     maxLinearAcceleration=700
     maxLinearVelocity=500
     velocityScaleZ=0
     worldAlpha=0.2
     externalLinearDragV=<0.2 0.2>
    }
   }
   n12 = "Transform (Modify) Bone - Bone: deform_glute_base_l"!(
    ComponentPose=@n11
    bAlphaBoolEnabled:bool=Get CrouchStill()
    Translation="-17.241,9.518,-0.319"
    Rotation="0,0,0"
    Scale="2.224,1.719,1.000"
   ) -> (Pose) {
    alphaInputType=Bool
    rotationMode=BMM_Additive
    rotationSpace=BCS_BoneSpace
    scaleMode=BMM_Additive
    scaleSpace=BCS_BoneSpace
    translationMode=BMM_Additive
    translationSpace=BCS_BoneSpace
    alphaBoolBlend {blendInTime=0.1, blendOption=Linear, blendOutTime=0.2}
    boneToModify {boneName=deform_glute_base_l}
    scale=<2.224 1.719>
    translation=<-17.241 9.518 -0.319>
   }
   n13 = Component To Local(ComponentPose=@n12)
   n14 = Get bIsCrouching()
   n15 = "Constraint - dyn_belt_MSTR (PPPPPPPPPP)"!(
    ComponentPose=Local To Component(LocalPose=Use cached pose 'ControlRig'())
    Alpha=1
    ConstraintWeights_0=0.2894
    ConstraintWeights_1:float=0.4579
    ConstraintWeights_2:float=0
    ConstraintWeights_3:float=0.2491
    ConstraintWeights_4:float=0.0015
    ConstraintWeights_5:float=0.0002
    ConstraintWeights_6:float=0
    ConstraintWeights_7:float=0.0005
    ConstraintWeights_8:float=0.0002
    ConstraintWeights_9:float=0.0012
   ) -> (Pose) {
    alphaBoolBlend {blendOption=Linear}
    boneToModify {boneName=dyn_belt_MSTR}
    constraintSetup [
     {targetBone {boneName=pelvis}}
     {targetBone {boneName=thigh_twist_01_l}}
     {targetBone {boneName=deform_glute_l}}
     {targetBone {boneName=deform_groin_l}}
     {targetBone {boneName=thigh_twist_01_r}}
     {targetBone {boneName=spine_01}}
     {targetBone {boneName=spine_03}}
     {targetBone {boneName=spine_02}}
     {targetBone {boneName=thigh_l}}
     {targetBone {boneName=deform_groin_r}}
    ]
    constraintWeights [1,1,1,1,1,1,1,1,1,1]
   }
   n16 = "Constraint - dyn_belt_02 (PPPPPPPP)"!(
    ComponentPose=@n15
    Alpha=1
    ConstraintWeights_0=0.1365
    ConstraintWeights_1:float=0.68
    ConstraintWeights_2:float=0.0002
    ConstraintWeights_3:float=0.155
    ConstraintWeights_4:float=0.0108
    ConstraintWeights_5:float=0.0092
    ConstraintWeights_6:float=0.0083
    ConstraintWeights_7:float=0
   ) -> (Pose) {
    alphaBoolBlend {blendOption=Linear}
    boneToModify {boneName=dyn_belt_02}
    constraintSetup [
     {targetBone {boneName=pelvis}}
     {targetBone {boneName=thigh_twist_01_l}}
     {targetBone {boneName=thigh_r}}
     {targetBone {boneName=deform_groin_l}}
     {targetBone {boneName=thigh_twist_01_r}}
     {targetBone {boneName=thigh_l}}
     {targetBone {boneName=deform_groin_r}}
     {targetBone {boneName=deform_glute_r}}
    ]
    constraintWeights [1,1,1,1,1,1,1,1]
   }
   n17 = "Constraint - dyn_belt_03 (PPPPPPPP)"!(
    ComponentPose=@n16
    Alpha=1
    ConstraintWeights_0=0.0186
    ConstraintWeights_1:float=0.8675
    ConstraintWeights_2:float=0.0009
    ConstraintWeights_3:float=0.0218
    ConstraintWeights_4:float=0.0118
    ConstraintWeights_5:float=0.0745
    ConstraintWeights_6:float=0.0047
    ConstraintWeights_7:float=0.0002
   ) -> (Pose) {
    alphaBoolBlend {blendOption=Linear}
    boneToModify {boneName=dyn_belt_03}
    constraintSetup [
     {targetBone {boneName=pelvis}}
     {targetBone {boneName=thigh_twist_01_l}}
     {targetBone {boneName=thigh_r}}
     {targetBone {boneName=deform_groin_l}}
     {targetBone {boneName=thigh_twist_01_r}}
     {targetBone {boneName=thigh_l}}
     {targetBone {boneName=deform_groin_r}}
     {targetBone {boneName=deform_glute_r}}
    ]
    constraintWeights [1,1,1,1,1,1,1,1]
   }
   n18 = "Transform (Modify) Bone - Bone: dyn_belt_01"(ComponentPose=@n17, Alpha=1, Translation="-4.431,0.823,1", Rotation="-1.669,0.541,0.615", Scale="1,1,1") {
    rotationMode=BMM_Additive
    rotationSpace=BCS_BoneSpace
    scaleMode=BMM_Additive
    scaleSpace=BCS_BoneSpace
    translationMode=BMM_Additive
    translationSpace=BCS_BoneSpace
    alphaBoolBlend {blendOption=Linear}
    boneToModify {boneName=dyn_belt_01}
    rotation=<rot -1.669 0.5411 0.615>
    translation=<-4.431 0.823 1>
   }
   n19 = "Transform (Modify) Bone - Bone: dyn_belt_02"(ComponentPose=@n18, Alpha=1, Translation="4,0.818,0.081", Rotation="1.582,0.898,13.128", Scale="1,1,1") {
    rotationMode=BMM_Additive
    rotationSpace=BCS_BoneSpace
    scaleMode=BMM_Additive
    scaleSpace=BCS_BoneSpace
    translationMode=BMM_Additive
    translationSpace=BCS_BoneSpace
    alphaBoolBlend {blendOption=Linear}
    boneToModify {boneName=dyn_belt_02}
    rotation=<rot 1.582 0.898 13.128>
    translation=<4 0.818 0.081>
   }
   n20 = Anim Dynamics - Bone: dyn_belt_02(ComponentPose=@n19, Alpha=1, ExternalForce="0,0,0") {
    angularDampingOverride=0.9999
    bOverrideAngularDamping=true
    bUseGravityOverride=true
    simulationSpace=RootRelative
    alphaBoolBlend {blendOption=Linear}
    boundBone {boneName=dyn_belt_02}
    gravityOverride {z=-980}
    physicsBodyDefinitions [{
      boundBone {boneName=dyn_belt_02}
      boxExtents=<8 18 18>
      constraintSetup {angularLimitsMax=<30 0 30>, angularLimitsMin=<-30 -30 -30>, angularTarget=<0 0 0>}
      localJointOffset=<3 0 0>
     }]
    relativeSpaceBone {boneName=thigh_twist_01_l}
   }
   n21 = Anim Dynamics - Bone: dyn_belt_03(ComponentPose=@n20, Alpha=1, ExternalForce="0,0,0") {
    angularDampingOverride=0.9999
    bOverrideAngularDamping=true
    bUseGravityOverride=true
    simulationSpace=RootRelative
    alphaBoolBlend {blendOption=Linear}
    boundBone {boneName=dyn_belt_03}
    componentAppliedLinearAccClamp=<5000 5000 5000>
    componentLinearAccScale=<1 1 1>
    componentLinearVelScale=<1 1 1>
    gravityOverride {z=-980}
    physicsBodyDefinitions [{
      boundBone {boneName=dyn_belt_03}
      boxExtents=<8 18 18>
      constraintSetup {angularLimitsMax=<30 0 30>, angularLimitsMin=<-30 -30 -30>, angularTarget=<0 0 0>}
      localJointOffset=<3 0 0>
     }]
    relativeSpaceBone {boneName=thigh_twist_01_l}
   }
   n22 = "Constraint - dyn_belt_MSTR (PPPPPPPP)"(
    ComponentPose=Local To Component(LocalPose=Use cached pose 'ControlRig'())
    bAlphaBoolEnabled=True
    ConstraintWeights_0=0.2505
    ConstraintWeights_1=0.4852
    ConstraintWeights_2=0.0022
    ConstraintWeights_3=0.001
    ConstraintWeights_4=0.134
    ConstraintWeights_5=0.0615
    ConstraintWeights_6=0.0146
    ConstraintWeights_7=0.0511
   ) {
    alphaInputType=Bool
    alphaBoolBlend {blendOption=Linear}
    boneToModify {boneName=dyn_belt_MSTR}
    constraintSetup [
     {targetBone {boneName=pelvis}}
     {targetBone {boneName=thigh_twist_01_l}}
     {targetBone {boneName=thigh_r}}
     {targetBone {boneName=deform_glute_l}}
     {targetBone {boneName=deform_groin_l}}
     {targetBone {boneName=thigh_twist_01_r}}
     {targetBone {boneName=thigh_l}}
     {targetBone {boneName=deform_groin_r}}
    ]
    constraintWeights [1,1,1,1,1,1,1,1]
   }
   n23 = "Transform (Modify) Bone - Bone: dyn_belt_MSTR"(ComponentPose=@n22, bAlphaBoolEnabled=Get CrouchStill(), Translation="5.738,4.937,0.369", Rotation="-31.741,37.399,43.934", Scale="1,1,1") {
    alphaInputType=Bool
    rotationMode=BMM_Additive
    rotationSpace=BCS_BoneSpace
    scaleMode=BMM_Additive
    scaleSpace=BCS_BoneSpace
    translationMode=BMM_Additive
    translationSpace=BCS_BoneSpace
    alphaBoolBlend {blendInTime=0.1, blendOption=Linear, blendOutTime=0.1}
    boneToModify {boneName=dyn_belt_MSTR}
    rotation=<rot -31.741 37.399 43.934>
    translation=<5.738 4.937 0.369>
   }
   n24 = "Transform (Modify) Bone - Bone: dyn_belt_MSTR"(ComponentPose=@n23, bAlphaBoolEnabled=Get Moving Crouch(), Translation="4.029,4.57,-1.051", Rotation="17.485,25.507,47.599", Scale="1,1,1") {
    alphaInputType=Bool
    rotationMode=BMM_Additive
    rotationSpace=BCS_BoneSpace
    scaleMode=BMM_Additive
    scaleSpace=BCS_BoneSpace
    translationMode=BMM_Additive
    translationSpace=BCS_BoneSpace
    alphaBoolBlend {blendInTime=0.1, blendOption=Linear, blendOutTime=0.1}
    boneToModify {boneName=dyn_belt_MSTR}
    rotation=<rot 17.485 25.507 47.599>
    translation=<4.029 4.57 -1.051>
   }
   n25 = RigidBody(ComponentPose=@n24, OverrideWorldGravity="0,0,-980") {
    bClampLinearTranslationLimitToRefPose=true
    bOverrideWorldGravity=true
    overridePhysicsAsset=&Pencil_Athletic_MedA_Lbdy_Physics_Crouching
    simulationSpace=BaseBoneSpace
    alphaBoolBlend {blendOption=Linear}
    baseBoneRef {boneName=pelvis}
    componentAppliedLinearAccClamp=<500 500 500>
    componentLinearAccScale=<0.3 0.3 0.3>
    overrideWorldGravity {z=-980}
    simSpaceSettings {
     maxAngularAcceleration=1000
     maxAngularVelocity=3
     maxLinearAcceleration=700
     maxLinearVelocity=500
     velocityScaleZ=0
     worldAlpha=0.2
     externalLinearDragV=<0.2 0.2>
    }
   }
   n26 = "Transform (Modify) Bone - Bone: deform_glute_base_l"(ComponentPose=@n25, bAlphaBoolEnabled=Get CrouchStill(), Translation="-17.241,9.518,-0.319", Rotation="0,0,0", Scale="2.224,1.719,1.000") {
    alphaInputType=Bool
    rotationMode=BMM_Additive
    rotationSpace=BCS_BoneSpace
    scaleMode=BMM_Additive
    scaleSpace=BCS_BoneSpace
    translationMode=BMM_Additive
    translationSpace=BCS_BoneSpace
    alphaBoolBlend {blendInTime=0.1, blendOption=Linear, blendOutTime=0.2}
    boneToModify {boneName=deform_glute_base_l}
    scale=<2.224 1.719>
    translation=<-17.241 9.518 -0.319>
   }
   n27 = Blend Poses by bool(
    BlendPose_0=Component To Local(ComponentPose=@n26)
    BlendPose_1=Component To Local(ComponentPose=@n21)
    BlendTime_0=0.2
    BlendTime_1=0.2
    bActiveValue=Get bIsCrouching()
   ) {blendType=Linear}
   n28 = "Constraint - dyn_belt_MSTR (PPPPPP)"(
    ComponentPose=Local To Component(LocalPose=Use cached pose 'ControlRig'())
    Alpha=1
    ConstraintWeights_0=0.7718
    ConstraintWeights_1=0.0573
    ConstraintWeights_2=0.0694
    ConstraintWeights_3=0.0347
    ConstraintWeights_4=0.0232
    ConstraintWeights_5=0.0437
   ) {
    alphaBoolBlend {blendOption=Linear}
    boneToModify {boneName=dyn_belt_MSTR}
    constraintSetup [
     {targetBone {boneName=pelvis}}
     {targetBone {boneName=thigh_twist_01_l}}
     {targetBone {boneName=deform_groin_l}}
     {targetBone {boneName=spine_01}}
     {targetBone {boneName=spine_03}}
     {targetBone {boneName=spine_02}}
    ]
    constraintWeights [1,1,1,1,1,1]
   }
   n29 = "Constraint - dyn_belt_MSTR (R)"(ComponentPose=@n28, Alpha=1, ConstraintWeights_0=1) {
    alphaBoolBlend {blendOption=Linear}
    boneToModify {boneName=dyn_belt_MSTR}
    constraintSetup [{transformType=Rotation, targetBone {boneName=thigh_l}}]
    constraintWeights [1]
   }
   n30 = RigidBody(ComponentPose=@n29, OverrideWorldGravity="0,0,-980") {
    bClampLinearTranslationLimitToRefPose=true
    bOverrideWorldGravity=true
    overridePhysicsAsset=&Pencil_Athletic_MedA_Lbdy_Physics_FrontEnd
    simulationSpace=BaseBoneSpace
    alphaBoolBlend {blendOption=Linear}
    baseBoneRef {boneName=root}
    componentAppliedLinearAccClamp=<5000 5000 5000>
    componentLinearAccScale=<0.4 0.4 0.5>
    componentLinearVelScale=<0.4 0.4 0.5>
    overrideWorldGravity {z=-980}
    simSpaceSettings {
     maxAngularAcceleration=1000
     maxAngularVelocity=3
     maxLinearAcceleration=700
     maxLinearVelocity=500
     velocityScaleZ=0
     externalLinearDragV=<0.2 0.2>
    }
   }
   n31 = "Constraint - dyn_belt_01 (P)"!(ComponentPose=@n30, bAlphaBoolEnabled:bool=True, ConstraintWeights_0=1) -> (Pose) {
    alphaInputType=Bool
    alphaBoolBlend {blendInTime=0.2, blendOption=Linear, blendOutTime=0.2}
    boneToModify {boneName=dyn_belt_01}
    constraintSetup [{targetBone {boneName=thigh_twist_01_l}}]
    constraintWeights [1]
   }
   n32 = "Constraint - dyn_belt_02 (P)"!(ComponentPose=@n31, bAlphaBoolEnabled:bool=True, ConstraintWeights_0=1) -> (Pose) {
    alphaInputType=Bool
    alphaBoolBlend {blendInTime=0.2, blendOption=Linear, blendOutTime=0.2}
    boneToModify {boneName=dyn_belt_02}
    constraintSetup [{targetBone {boneName=thigh_twist_01_l}}]
    constraintWeights [1]
   }
   n33 = Component To Local(ComponentPose=@n32)
   n34 = Get bIsInFrontEnd()
   Output Pose(Result=Blend Poses by bool(
     BlendPose_0=Blend Poses by bool(BlendPose_0=@n33, BlendPose_1=@n27, BlendTime_0=0.2, BlendTime_1=0.2, bActiveValue=@n34) {blendType=Linear}
     BlendPose_1=Blend Poses by bool(BlendPose_0=@n13, BlendPose_1=@n7, BlendTime_0=0, BlendTime_1=0, bActiveValue=@n14) {blendType=Linear}
     bActiveValue=Get Jacket() -> (Jacket:bool)
    ) {blendType=Linear}) {layerGroup=None}
   Save cached pose 'ControlRig'(Pose=Copy Pose From Mesh() {bUseAttachedParent=true}) {cacheName=ControlRig}
  ]}
 EventGraph {nodes [
   Event BlueprintUpdateAnimation()
   > n1 = Sequence!(execute) -> (then_0, then_1, then_2:Exec, then_3:Exec, then_4:Exec)
   > n2 = Gravity Override State Function DEPRECATED(CharacterPartAnimInstance=Self-Reference(), GameTimeInSeconds=GetGameTimeInSeconds(), GravityOverrideParams=Get Gravity Override Params())
   > Set GravityOverride(GravityOverride=@n2)
   Set Jacket(
    execute=@n1.then_1
    Jacket:bool=CharacterItemDefinitionHasAnyMetaTag(Tags=Make Array("[0]"=(TagName="Cosmetics.MetaTags.Pencil.Upper.Jacket")))
   ) -> (Output_Get:bool)
   n3 = Branch(execute=@n1.then_2, Condition=AND Boolean(A=Get bIsCrouching(), B=Get bIsAccelerating2D()))
   > Set Moving Crouch(Moving Crouch:bool=true) -> (Output_Get:bool)
   Set Moving Crouch(execute=@n3.else, Moving Crouch=false)
   n4 = Get bIsAccelerating2D()
   n5 = Branch(execute=@n1.then_3, Condition=AND Boolean(A=Get bIsCrouching(), B=NOT Boolean(A=@n4)))
   > Set CrouchStill(CrouchStill:bool=true) -> (Output_Get:bool)
   Set CrouchStill(execute=@n5.else, CrouchStill=false)
  ]}
]
variables [
 CrouchStill {type=boolean}
 Gravity Override Params {type=GravityOverrideParamsStruct}
 GravityOverride {type=Vector}
 Jacket {type=boolean}
 Moving Crouch {type=boolean}
]
